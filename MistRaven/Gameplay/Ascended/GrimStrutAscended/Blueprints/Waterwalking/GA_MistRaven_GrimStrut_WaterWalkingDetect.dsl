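; Gameplay ability for Grim Strut water walking: it makes water bodies solid under the player by switching the capsule's water trace channel to Block.
; A pawn in non-ocean water is pushed upward with root motion force until it exits, then collision blocks and a looping timer updates the water walking GameplayEffect.
; Settings can let ocean or long falls switch to overlap, while ice sliding restores blocking.
; EndAbility restores overlap and cleans up tasks, timers and bindings.
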(event Ability|EventActivateAbility
  ;[0 0.15 0.02036 0.5] ActivateAbility: Handle player being in or out of water. If in water kick them out, if out of water disable collision
  (bind _outfortplayerpawn (CallFunction|GetAvatarasValidFortPlayerPawn)
    (:ValidFortPlayerPawn
      (bind _output_get (Variables|Transient|SetAvatarPlayerPawn _outfortplayerpawn))
      (if (ScalableFloat|GetValueasBool (Variables|Settings|GetRow_CanFallIntoWaterEnabled))
        (Character|BindEventtoMovementModeChangedDelegate _output_get (EventDispatchers|CreateEvent :!event Custom|Movement%20Changed)) ; Handling falling into water
        (else
          _))
      (CallFunction|BlockWalkingonOcean)
      (Utilities|IsinWater _outfortplayerpawn
        (:In%20Ocean
          (if (ScalableFloat|GetValueasBool (Variables|Settings|GetRow_DisableOceanWalking))
            (EventDispatchers|BindEventtoOnPawnEnterWater _output_get (EventDispatchers|CreateEvent :!event Custom|HandlePawnInWater))
            (Variables|Transient|SetbIgnoringOcean true)
            (else
              _)))
        (:In%20Non-Ocean
          _)
        (:Not%20In%20Water
          (if (and (Utilities|Enum|Equal(Enum) (Variables|/Script/Engine.CharacterMovementComponent|GetMovementMode (Variables|/Script/Engine.Character|GetCharacterMovement _output_get)) "MOVE_Falling") (ScalableFloat|GetValueasBool (Variables|Settings|GetRow_CanFallIntoWaterEnabled)))
            (EventDispatchers|BindEventtoOnPawnEnterWater _output_get (EventDispatchers|CreateEvent :!event Custom|HandlePawnInWater))
            (CallFunction|CheckForIceSliding)
            (else
              (CallFunction|HandlePawnExitWaterBody (CallFunction|GetAvatarasFortPlayerPawn) true)))))
      (CallFunction|HandlePawninWater _output_get))))

(event Custom|Movement%20Changed
  (params (Character :type Character%20Object%20Reference) (PrevMovementMode :type EMovementMode%20Enum :default "MOVE_None") (PreviousCustomMode :type Byte))
  ;[0.02627 0.06212 0.15 0.5] Checking if the player fell into water from a long enough fall to potentially deal damage
  (bind _asfort_player_pawn_athena (Utilities|Casting|CastToFortPlayerPawnAthena Character))
  ;; Check if the player has started falling
  (if (and (not (Utilities|Time|IsValidTimerHandle (Variables|Transient|GetFallingTimer))) (Utilities|Enum|Equal(Enum) (Variables|/Script/Engine.CharacterMovementComponent|GetMovementMode (Variables|/Script/Engine.Character|GetCharacterMovement _asfort_player_pawn_athena)) "MOVE_Falling"))
    ;; This timer will not always provide the most accurate time for when fall damage will happen. Fall damage is calculated by distance from starting point of the fall. This solution covers edge cases for when the weapon gets equipped and is generally safe, but a cleaner, more thorough solution that uses the actual calculation for fall damage would be more correct here.
    (bind _returnvalue (Utilities|Time|SetTimerbyEvent (EventDispatchers|CreateEvent :!event Custom|Prevent%20Water%20Fall%20Damage) (ScalableFloat|GetValueasDouble (Variables|Settings|GetMinHarmfulFallDuration))))
    (Variables|Transient|SetFallingTimer _returnvalue)
    (CallFunction|CheckForIceSliding)
    (else
      ;; If the player is not in danger of taking fall damage, re-enable water walking
      (Utilities|Time|ClearandInvalidateTimerbyHandle (Variables|Transient|GetFallingTimer))
      (if (not (Variables|Transient|GetbIgnoringOcean))
        (bind _avatarplayerpawn (Variables|Transient|GetAvatarPlayerPawn)
          (:then
            (Utilities|IsinWater _avatarplayerpawn
              (:Not%20In%20Water
                (CallFunction|BlockWater _avatarplayerpawn))))
          (:else))))))

(event Custom|Check%20for%20Ice%20Sliding
  ;[0.06217 0.02382 0.15 0.5] Starts a Wait to see if the player has started Ice Sliding
  (bind _avatarplayerpawn (Variables|Transient|GetAvatarPlayerPawn)
    (:then
      (bind _asynctaskproxy (Ability|Tasks|WaitGameplayTagAdd (Variables|Transient|GetTag_IceSliding) _avatarplayerpawn)
        (:then
          (Variables|Transient|SetAsyncTask_BlockWaterForSlide _asynctaskproxy))
        (:Added
          (CallFunction|BlockWaterforSlide))))
    (:else)))

(event Custom|Block%20Water%20For%20Slide
  ;[0.1226 0 0.15 0.5] Sets the water collision back to blocking and invalidates the fall timer if the player begins sliding
  (Utilities|Time|ClearandInvalidateTimerbyHandle (Variables|Transient|GetFallingTimer))
  (bind _avatarplayerpawn (Variables|Transient|GetAvatarPlayerPawn)
    (:then
      (if (and (Utilities|Enum|NotEqual(Enum) (Collision|GetCollisionResponsetoChannel (Variables|/Script/Engine.Character|GetCapsuleComponent _avatarplayerpawn) "ECC_GameTraceChannel18") "ECR_Block") (not (Variables|Transient|GetbIgnoringOcean)))
        (CallFunction|BlockWater _avatarplayerpawn)))
    (:else)))

(event Custom|Prevent%20Water%20Fall%20Damage
  ;[0.03599 0.0262 0.15 0.5] Triggers when falling long enough to potentially take fall damage. Sets water collision to overlap.
  (bind _avatarplayerpawn (Variables|Transient|GetAvatarPlayerPawn)
    (:then
      (if (not (GameplayTags|HasMatchingGameplayTag _avatarplayerpawn (Variables|Transient|GetTag_IceSliding)))
        (Collision|SetCollisionResponsetoChannel (Variables|/Script/Engine.Character|GetCapsuleComponent _avatarplayerpawn) "ECC_GameTraceChannel18" "ECR_Overlap")
        (EventDispatchers|BindEventtoOnPawnEnterWater _avatarplayerpawn (EventDispatchers|CreateEvent :!event Custom|HandlePawnInWater))
        (else
          (CallFunction|BlockWaterforSlide))))
    (:else)))

(event Custom|HandlePawnInWater
  (params (FortPlayerPawn :type Fort%20Player%20Pawn%20Object%20Reference))
  ;[0.04418 0.15 0.08139 0.5] Push player out of water and turn on water collision once out of water. Ignore if water body is an ocean
  (Utilities|IsValid FortPlayerPawn
    (:Is%20Valid
      (Utilities|IsinWater FortPlayerPawn
        (:In%20Non-Ocean
          (Variables|Transient|SetbPawnInWaterBody true)
          (Variables|Transient|SetPawnInWaterBodyImmersionDepth (WaterBody|GetWaterImmersionDepth FortPlayerPawn))
          (EventDispatchers|BindEventtoOnPawnExitWater FortPlayerPawn (EventDispatchers|CreateEvent :!event Custom|HandlePawnExitWaterBody))
          (EventDispatchers|UnbindEventfromOnPawnEnterWater FortPlayerPawn (EventDispatchers|CreateEvent :!event Custom|HandlePawnInWater))
          ;; Player is in water, apply constant force for a bit to push them out. Try again if still in water after
          (bind _asynctaskproxy (Ability|Tasks|ApplyRootMotionConstantForce "Task_ForceOutOfWaterConstant" "0.000000,0.000000,1.000000" (Variables|Settings|GetForceOutOfWaterStrength) (Variables|Settings|GetForceOutOfWaterDuration) true)
            (:then
              (Variables|Transient|SetAsyncTask_ForceOutOfWaterConstant _asynctaskproxy))
            (:OnFinish
              (bind _avatarplayerpawn (Variables|Transient|GetAvatarPlayerPawn)
                (:then
                  (Utilities|IsValid (WaterBody|GetCurrentWaterBody _avatarplayerpawn)
                    (:Is%20Valid
                      (CallFunction|HandlePawninWater _avatarplayerpawn))))
                (:else)))))))))

(event Custom|HandlePawnExitWaterBody
  (params (FortPlayerPawn :type Fort%20Player%20Pawn%20Object%20Reference) (bPawnIsOutsideOfAllWaterBodies :type Boolean))
  ;[0.1046 0.04605 0.15 0.5] HandlePawnExitWaterBody: The player has exited water body, change their collision relationship with the waterbody now.
  (if bPawnIsOutsideOfAllWaterBodies
    (EventDispatchers|UnbindEventfromOnPawnExitWater FortPlayerPawn (EventDispatchers|CreateEvent :!event Custom|HandlePawnExitWaterBody))
    (bind _avatarplayerpawn (Variables|Transient|GetAvatarPlayerPawn)
      (:then
        (GameplayTasks|EndTask (Variables|Transient|GetAsyncTask_ForceOutOfWaterConstant))
        (CallFunction|BlockWater _avatarplayerpawn))
      (:else))))

(event Custom|BlockWater
  (params (Player :type Fort%20Pawn%20Object%20Reference))
  ;[0.1195 0.03101 0.15 0.5] Set the collision on water bodies to blocking
  (Variables|Transient|SetbIgnoringOcean false)
  (Collision|SetCollisionResponsetoChannel (Variables|/Script/Engine.Character|GetCapsuleComponent Player) "ECC_GameTraceChannel18" "ECR_Block")
  (bind _returnvalue (Utilities|Time|SetTimerbyEvent (EventDispatchers|CreateEvent :!event Custom|UpdateWaterWalkingStatus) (Variables|Settings|GetWaterWalkingStatusUpdateInterval) true true))
  (Variables|Transient|SetUpdateWaterWalkingStatusTimerHandle _returnvalue)
  (CallFunction|UpdateWaterWalkingStatus)
  (bind _asynctask_blockwaterforslide (Variables|Transient|GetAsyncTask_BlockWaterForSlide)
    (:then
      (GameplayTasks|EndTask _asynctask_blockwaterforslide))
    (:else
      _))
  (Utilities|IsinWater Player
    (:In%20Non-Ocean
      (Ability|ExecuteGameplayCueWithParamsOnOwner (Variables|Settings|GetGameplayCueWaterWalkingTransitionApplication) (Ability|GameplayCue|MakeGameplayCueParameters :NormalizedMagnitude 0.0 :RawMagnitude (Variables|Transient|GetPawnInWaterBodyImmersionDepth) :Location "0,0,0" :Normal "0,0,0" :SourceObject (Variables|Transient|GetPawnInWaterBodyActor) :GameplayEffectLevel 1 :AbilityLevel 1 :bReplicateLocationWhenUsingMinimalRepProxy false)))))

(event Custom|BlockWalkingOnOcean
  ;[0.15 0.06181 0.01223 0.5] Disable water collision if movement base is ocean. Wait for entering water body so collision can be re-enabled when entering non-ocean water
  (if (ScalableFloat|GetValueasBool (Variables|Settings|GetRow_DisableOceanWalking))
    (bind (_asynctaskproxy _tagcount) (Ability|Tasks|WaitGameplayTagCountChange "(TagName=\"Gameplay.MovementMode.MovementIsIgnoringOceanBase\")")
      (:then
        (Variables|Transient|SetWaitForOceanWalkTask _asynctaskproxy)
        (Variables|Transient|SetbIgnoringOcean false))
      (:TagCountChanged
        (if (> _tagcount 0)
          (bind _avatarplayerpawn (Variables|Transient|GetAvatarPlayerPawn)
            (:then
              (Collision|SetCollisionResponsetoChannel (Variables|/Script/Engine.Character|GetCapsuleComponent _avatarplayerpawn) "ECC_GameTraceChannel18" "ECR_Overlap")
              (Variables|Transient|SetbIgnoringOcean true)
              (EventDispatchers|BindEventtoOnPawnEnterWater _avatarplayerpawn (EventDispatchers|CreateEvent :!event Custom|HandlePawnInWater)))
            (:else)))))))

(event Custom|UpdateWaterWalkingStatus
  ;[0.15 0.07194 0.03994 0.5] UpdateWaterWalkingStatus: Update the water walking status on the player pawn.
  (bind _avatarplayerpawn (Variables|Transient|GetAvatarPlayerPawn)
    (:then
      (Utilities|IsValid (WaterBody|GetCurrentWaterBody _avatarplayerpawn)
        (:Is%20Valid
          (CallFunction|HandlePawninWater 0)
          (Utilities|Time|ClearandInvalidateTimerbyHandle (Variables|Transient|GetUpdateWaterWalkingStatusTimerHandle)))
        (:Is%20Not%20Valid
          (bind _charactermovement (Variables|/Script/Engine.Character|GetCharacterMovement _avatarplayerpawn)
            (:then
              (bind _asprimitive_component (Utilities|Casting|CastToPrimitiveComponent (Pawn|Components|CharacterMovement|GetMovementBaseObject _charactermovement))
                (:then
                  (if (.bSuccess (Utilities|Casting|CastToFortWaterBodyActor (Components|GetOwner _asprimitive_component)))
                    (Ability|ApplyGameplayEffectToOwner "/MistRaven/Gameplay/Ascended/GrimStrutAscended/Blueprints/Waterwalking/GE_MistRaven_GrimStrut_WaterWalking.GE_MistRaven_GrimStrut_WaterWalking_C")
                    (else
                      _)))
                (:CastFailed
                  _))
              (Ability|RemoveGameplayEffectFromOwnerWithGrantedTags "(GameplayTags=((TagName=\"MotoWrestle.Granted.Status.WaterWalking\")))"))
            (:else)))))
    (:else)))

(event Ability|EventOnEndAbility
  (params (bWasCancelled :type Boolean))
  ;[0.15 0.09274 0.002054 0.5] EndAbility: Clean up.
  (Utilities|FlowControl|Sequence
    (:then_0
      (Variables|Transient|SetbPawnInWaterBody false)
      (Variables|Transient|SetPawnInWaterBodyImmersionDepth 0.0)
      (Variables|Transient|SetPawnInWaterBodyActor 0)
      (Variables|Transient|SetbIgnoringOcean false)
      (Ability|RemoveGameplayEffectFromOwnerWithGrantedTags "(GameplayTags=((TagName=\"MotoWrestle.Granted.Status.WaterWalking\")))")
      (bind _avatarplayerpawn (Variables|Transient|GetAvatarPlayerPawn)
        (:then
          (Collision|SetCollisionResponsetoChannel (Variables|/Script/Engine.Character|GetCapsuleComponent _avatarplayerpawn) "ECC_GameTraceChannel18" "ECR_Overlap")
          (Ability|RemoveGameplayEffectFromOwnerWithHandle (Variables|Transient|GetWaterWalkingStatusGameplayEffectHandle))
          (bind _charactermovement (Variables|/Script/Engine.Character|GetCharacterMovement _avatarplayerpawn)
            (:then
              (bind _asprimitive_component (Utilities|Casting|CastToPrimitiveComponent (Pawn|Components|CharacterMovement|GetMovementBaseObject _charactermovement)))
              (bind _asfort_water_body_actor (Utilities|Casting|CastToFortWaterBodyActor (Components|GetOwner _asprimitive_component)))
              (bind _waterbodycomponent (Variables|/Script/Water.WaterBody|GetWaterBodyComponent _asfort_water_body_actor)
                (:then
                  (Ability|ExecuteGameplayCueWithParamsOnOwner (Variables|Settings|GetGameplayCueWaterWalkingTransitionRemoval) (Ability|GameplayCue|MakeGameplayCueParameters :NormalizedMagnitude 0.0 :RawMagnitude (.OutWaterDepth (WaterBody|GetWaterSurfaceInfoatLocation _waterbodycomponent (Math|Vector|MakeVector (.x (Transformation|GetActorLocation _avatarplayerpawn)) (.y (Transformation|GetActorLocation _avatarplayerpawn)) (- (.z (Transformation|GetActorLocation _avatarplayerpawn)) (.OutHalfHeight (Components|Capsule|GetScaledCapsuleSize (Variables|/Script/Engine.Character|GetCapsuleComponent _avatarplayerpawn))))) true)) :Location "0,0,0" :Normal "0,0,0" :SourceObject _asfort_water_body_actor :GameplayEffectLevel 1 :AbilityLevel 1 :bReplicateLocationWhenUsingMinimalRepProxy false))
                  (Pawn|Character|LaunchCharacterJump _avatarplayerpawn "0.000000,0.000000,-1.000000" false false false true) ; Minor velocity to force movement update, so the pawn falls into the water.
                )
                (:else)))
            (:else)))
        (:else)))
    (:then_1
      (Utilities|Time|ClearandInvalidateTimerbyHandle (Variables|Transient|GetUpdateWaterWalkingStatusTimerHandle))
      (bind _asynctask_forceoutofwaterconstant (Variables|Transient|GetAsyncTask_ForceOutOfWaterConstant)
        (:then
          (GameplayTasks|EndTask _asynctask_forceoutofwaterconstant))
        (:else
          _))
      (bind _waitforoceanwalktask (Variables|Transient|GetWaitForOceanWalkTask)
        (:then
          (GameplayTasks|EndTask _waitforoceanwalktask))
        (:else
          _))
      (bind _avatarplayerpawn_2 (Variables|Transient|GetAvatarPlayerPawn)
        (:then
          (EventDispatchers|UnbindEventfromOnPawnExitWater _avatarplayerpawn_2 (EventDispatchers|CreateEvent :!event Custom|HandlePawnExitWaterBody))
          (EventDispatchers|UnbindEventfromOnPawnEnterWater _avatarplayerpawn_2 (EventDispatchers|CreateEvent :!event Custom|HandlePawnInWater))
          (Character|UnbindEventfromMovementModeChangedDelegate _avatarplayerpawn_2 (EventDispatchers|CreateEvent :!event Custom|Movement%20Changed))
          (Utilities|Time|InvalidateTimerHandle (Variables|Transient|GetFallingTimer))
          (bind _asynctask_blockwaterforslide (Variables|Transient|GetAsyncTask_BlockWaterForSlide)
            (:then
              (GameplayTasks|EndTask _asynctask_blockwaterforslide))
            (:else)))
        (:else)))
    (:then_2
      (|CallParentFunction bWasCancelled))))
